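# Animation Blueprint for the helmet part of the female medium Hazmat player cosmetic, driving its skeleton from another skeletal mesh.
# On initialization it caches the result of GetOwnerSkeletalMeshComponent in MeshToCopy.
# The AnimGraph feeds that component into a Copy Pose From Mesh node wired straight to the output pose, with no animation logic of its own.
# The update event is empty, and a TryGetPawnOwner node sits unused.
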AnimBlueprint /BRCosmetics/Characters/Player/Female/Medium/Bodies/F_MED_Hazmat/Meshes/Parts/F_MED_Hazmat_01_Helmet_Skeleton_AnimBlueprint
target_skeleton=&/F_MED_Hazmat_01_Helmet_Skeleton
graphs [
 AnimGraph {nodes [Output Pose(Result=Copy Pose From Mesh(SourceMeshComponent=Get MeshToCopy() -> (MeshToCopy:Skeletal Mesh Component Object Reference))) {layerGroup=None}]}
 EventGraph {nodes [
   Event BlueprintInitializeAnimation()
   > Set MeshToCopy(MeshToCopy:Skeletal Mesh Component Object Reference=GetOwnerSkeletalMeshComponent()) -> (Output_Get:Skeletal Mesh Component Object Reference)
   TryGetPawnOwner()
   Event BlueprintUpdateAnimation()
  ]}
]
variables [MeshToCopy {type=SkeletalMeshComponent}]
